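# Physics asset for the Garage Band mic stand instrument cosmetic, giving capsule bodies to its handle, dangling cloth pieces and wire.
# The cloth forms a wide-swinging A chain, with B and C pieces branching from its root, which has no constraint to the handle.
# Top and bottom wire chains hang from the handle with mostly tight swing and locked twist, and a constraint joins their ends into a closed loop.
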PhysicsAsset /SparksCosmetics/Instrument/MicStand/GarageBand/Meshes/Garage_Band_MicStand_Physics
bodies {
 dyn_cloth_A_01 [{shapeType=Capsule, length=0.5, radius=0.5, shapeName=dyn_cloth_A_01_capsule, rotation=<rot 90 0 0>}]
 dyn_cloth_A_02 [{shapeType=Capsule, length=3, radius=1.5, shapeName=dyn_cloth_A_02_capsule, center=<2.35 0.3 1>, rotation=<rot 90 -7.125 -17.1249>}]
 dyn_cloth_A_03 [{shapeType=Capsule, length=3, radius=1.75, shapeName=dyn_cloth_A_03_capsule, center=<1.8 0 0>, rotation=<rot 90 0 0>}]
 dyn_cloth_A_04 [{shapeType=Capsule, length=3, radius=2, shapeName=dyn_cloth_A_04_capsule, center=<1.8 0 0>, rotation=<rot 90 0 0>}]
 dyn_cloth_B [{shapeType=Capsule, length=5.5, radius=1, shapeName=dyn_cloth_B_capsule, center=<2.5 0 0>, rotation=<rot 90 0 0>}]
 dyn_cloth_C [{shapeType=Capsule, length=4, radius=0.8184, shapeName=dyn_cloth_C_capsule, center=<2.4 0 0.45>, rotation=<rot 90 0 0>}]
 handle [{shapeType=Capsule, length=71.0632, radius=1.3855, center=<0.02 -0.0518 17.4457>, rotation=<rot 0.3792 -90 -0.1451>}]
 wire_bottom_01 [{shapeType=Capsule, length=1, radius=1, shapeName=wire_bottom_01_capsule, rotation=<rot 90 0 0>}]
 wire_bottom_02 [{shapeType=Capsule, length=3.5477, radius=1, shapeName=wire_bottom_02_capsule, rotation=<rot 80.1728 -173.8405 -172.9186>}]
 wire_bottom_connector [{shapeType=Capsule, length=4.207, radius=1, shapeName=wire_bottom_connector_capsule, center=<-0.0028 0 0.0007>, rotation=<rot 76.1293 0 0>}]
 wire_top_01 [{shapeType=Capsule, length=5.8747, radius=1, shapeName=wire_top_01_capsule, center=<2.824 0 -1.1703>, rotation=<rot 67.4906 0 0>}]
 wire_top_02 [{shapeType=Capsule, length=6.9648, radius=1, shapeName=wire_top_02_capsule, rotation=<rot 71.9453 0 0>}]
 wiretop_top_03 [{shapeType=Capsule, length=4.2448, radius=1, shapeName=wiretop_top_03_capsule, rotation=<rot 80.0037 0 0>}]
}
constraints [
 {bone1Name=dyn_cloth_A_02, bone2Name=dyn_cloth_A_01, swing1LimitDegrees=80, swing2LimitDegrees=70, twistLimitDegrees=10}
 {bone1Name=dyn_cloth_A_03, bone2Name=dyn_cloth_A_02, swing1LimitDegrees=115, swing2LimitDegrees=115, twistLimitDegrees=5}
 {bone1Name=dyn_cloth_A_04, bone2Name=dyn_cloth_A_03, swing1LimitDegrees=115, swing2LimitDegrees=115, twistLimitDegrees=5}
 {bone1Name=dyn_cloth_B, bone2Name=dyn_cloth_A_01, swing1LimitDegrees=90, swing2LimitDegrees=60, twistLimitDegrees=2, twistMotion=Locked}
 {bone1Name=dyn_cloth_C, bone2Name=dyn_cloth_A_01, swing1LimitDegrees=60, swing2LimitDegrees=45, twistLimitDegrees=2, twistMotion=Locked}
 {bone1Name=wire_top_02, bone2Name=wire_top_01, swing1LimitDegrees=15, swing2LimitDegrees=5, twistLimitDegrees=45, twistMotion=Locked}
 {bone1Name=wiretop_top_03, bone2Name=wire_top_02, swing1LimitDegrees=5, swing2LimitDegrees=5, twistLimitDegrees=45, twistMotion=Locked}
 {bone1Name=wire_top_01, bone2Name=handle, swing1LimitDegrees=60, swing2LimitDegrees=10, twistLimitDegrees=45, twistMotion=Locked}
 {bone1Name=wire_bottom_01, bone2Name=handle, swing1LimitDegrees=30, swing2LimitDegrees=10, twistLimitDegrees=45, twistMotion=Locked}
 {bone1Name=wire_bottom_02, bone2Name=wire_bottom_01, swing1LimitDegrees=15, swing2LimitDegrees=10, twistLimitDegrees=45, twistMotion=Locked}
 {
  bone1Name=wire_bottom_connector
  bone2Name=wire_bottom_02
  swing1LimitDegrees=10
  swing2LimitDegrees=10
  twistLimitDegrees=45
  twistMotion=Locked
 }
 {
  bone1Name=wiretop_top_03
  bone2Name=wire_bottom_connector
  swing1LimitDegrees=5
  swing2LimitDegrees=5
  twistLimitDegrees=45
  twistMotion=Locked
 }
]
